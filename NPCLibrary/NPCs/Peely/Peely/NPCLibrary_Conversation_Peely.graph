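# Conversation database for the Peely NPC, whose dialogue consists of described gestures and moods rather than spoken lines.
# One entry point feeds a single speech node that picks contextual messages over a default shiver line.
# First-ever meetings and specific player outfits have named priority curves, while other variants check Peely's per-match conversation count.
# The speech node then links out to the shared NPCLibrary quests and basic services conversations.
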ConversationDatabase /NPCLibrary/NPCs/Peely/Peely/NPCLibrary_Conversation_Peely
entry_points [{tag=Conversation.Entry.Peely, destinations [#1]}]
nodes [
 ConversationEntryPointNode_7 {
  id=#1
  class=ConversationEntryPointNode
  outputs [#2]
  properties {entryTag {tagName=Conversation.Entry.Peely}}
 }
 Speech_C_4 {
  id=#2
  class=Speech_C
  outputs [#3, #4]
  properties {
   nodeName=""
   generalConfig {
    defaultMessage=(Peely shivers a bit, but pretends he isn't.)
    contextualMessages [
     {
      message=(Peely gives you a big, happy grin.)
      contextRequirements [{requirement {class=FortControllerRequirement_IsFirstEverConversationWithNPC}}]
      priority {curve {rowName=Tandem.Speech.Priority.FirstEverInteraction}}
     }
     {
      message=(Peely gives you a big hug, then shoos you away to your next big adventure.)
      contextRequirements [{requirement {
         class=FortControllerRequirement_HasCID
         properties {allowedCIds [&Character_UndergroundRebel_Fashion]}
        }}]
      priority {curve {rowName=Tandem.Speech.Priority.OutfitReaction}}
     }
     {
      message=(Peely grins and makes himself little cat ears with his fingers. It looks weird on a banana.)
      contextRequirements [{requirement {
         class=FortControllerRequirement_HasCID
         properties {allowedCIds [&Character_VoidRedemption_Rebel]}
        }}]
      priority {curve {rowName=Tandem.Speech.Priority.OutfitReaction}}
     }
     {
      message=(Peely scowls.)
      contextRequirements [{requirement {
         class=FortControllerRequirement_HasCID
         properties {allowedCIds [
           /BRCosmetics/Athena/Items/Cosmetics/Characters/Character_MadameMoth_Posh
           &Character_BengalBasher_Suit
           /BRCosmetics/Athena/Items/Cosmetics/Characters/Character_CeremonialGuard_Fencer
           /BRCosmetics/Athena/Items/Cosmetics/Characters/Character_DiamondHeart_Chic
          ]}
        }}]
      priority {curve {rowName=Tandem.Speech.Priority.OutfitReaction}}
     }
     {
      message=(Peely shivers a bit, but pretends he isn't.)
      contextRequirements [{requirement {class=FortControllerRequirement_CheckNPCMatchConversationCount, properties {nPCUniqueIdTag {tagName=AISpawnerData.Type.Tandem.Peely}}}}]
      priority {curve {rowName=None}}
     }
     {
      message=(Peely marvels at the beauty of nature around him.)
      contextRequirements [{requirement {class=FortControllerRequirement_CheckNPCMatchConversationCount, properties {nPCUniqueIdTag {tagName=AISpawnerData.Type.Tandem.Peely}}}}]
      priority {curve {rowName=None}}
     }
     {
      message=(Peely sighs, and looks at his little tent dejectedly.)
      contextRequirements [{requirement {class=FortControllerRequirement_CheckNPCMatchConversationCount, properties {nPCUniqueIdTag {tagName=AISpawnerData.Type.Tandem.Peely}}}}]
      priority {curve {rowName=None}}
     }
     {
      message=(Peely mimes driving and honking a horn.)
      contextRequirements [{requirement {class=FortControllerRequirement_CheckNPCMatchConversationCount, properties {nPCUniqueIdTag {tagName=AISpawnerData.Type.Tandem.Peely}}}}]
      priority {curve {rowName=None}}
     }
     {
      message=(Peely does a little dance, thankful to be free.)
      contextRequirements [{requirement {class=FortControllerRequirement_CheckNPCMatchConversationCount, properties {nPCUniqueIdTag {tagName=AISpawnerData.Type.Tandem.Peely}}}}]
      priority {curve {rowName=None}}
     }
    ]
   }
  }
 }
 ConversationLinkNode_0 {id=#4, class=ConversationLinkNode, properties {remoteEntryTag {tagName=Conversation.Entry.BasicServices}}}
 ConversationLinkNode_1 {id=#3, class=ConversationLinkNode, properties {remoteEntryTag {tagName=Conversation.Entry.NPCLibrary.Quests}}}
]
